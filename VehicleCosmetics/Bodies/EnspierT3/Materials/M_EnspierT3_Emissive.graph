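# Translucent, unlit, two-sided emissive material for the EnspierT3 vehicle body cosmetic, usable on skeletal meshes.
# Opacity combines panning noise, partly in screen space, with scrolling scan textures, masked by squared vertex color.
# Emissive is a boosted CustomColor parameter; an added reflection-vector term is multiplied by zero.
# A WPO reroute builds a small sine/cosine wobble from time, phase-offset by vertex alpha.
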Material /VehicleCosmetics/Bodies/EnspierT3/Materials/M_EnspierT3_Emissive
graph {
 properties {
  bAutomaticallySetUsageInEditor=false
  bUsedWithSkeletalMesh=true
  blendMode=BLEND_Translucent
  shadingModel=MSM_Unlit
  twoSided=true
  usedShadingModels=Unlit
 }
 nodes [
  n1 = Add(A=Time(), B=VertexColor().A)
  n2 = AppendVector(A=AppendVector(A=Sine(None=@n1) {period=2}, B=Cosine(None=@n1) {period=2}), B=Sine(None=@n1) {period=1.333})
  NamedRerouteDeclaration(None=Multiply(A=Transform(None=@n2), B=Constant3Vector() {constant=<rgba 0.1 0.1 0.1 1>})) {name=WPO, nodeColor=<rgba 0 0.25 1 1>}
  n3 = Divide(A=ComponentMask(None=ScreenPosition().ViewportUV) {g=true, r=true}, B=Constant() {r=0.5})
  n4 = Panner(Coordinate=ComponentMask(None=@n3) {g=true, r=true}) {speedY=0.05}
  n5 = TextureSample(UVs=Panner() {speedY=1}) {samplerSource=SSM_Wrap_WorldGroupSettings, texture=/VehicleCosmetics/Wheels/_SharedTextures/FX/Noise_Generic_01_Pack}
  n6 = TextureSample(UVs=Panner(Coordinate=ComponentMask(None=@n3) {g=true, r=true}) {speedY=-1}) {samplerSource=SSM_Wrap_WorldGroupSettings, texture=&Noise_Generic_01_Pack}
  n7 = Multiply(A=@n5.R, B=@n6.G)
  n8 = Multiply(A=@n5.G, B=@n6.R)
  n9 = Multiply(A=Multiply(
    A=TextureSample(UVs=@n4) {samplerSource=SSM_Wrap_WorldGroupSettings, texture=&Noise_Generic_01_Pack}.R
    B=Add(A=@n7, B=@n8)
   )) {constB=8}
  n10 = TextureCoordinate() {vTiling=16}
  n11 = TextureSample(UVs=Panner(Coordinate=ComponentMask(None=@n10) {g=true, r=true}) {speedY=0.1}) {samplerSource=SSM_Wrap_WorldGroupSettings, texture=/VehicleCosmetics/Wheels/_SharedTextures/FX/Scan_M}
  n12 = Panner(Coordinate=ComponentMask(None=TextureCoordinate()) {g=true, r=true}) {speedY=-0.15}
  n13 = LinearInterpolate(Alpha=Add(
    A=Multiply(A=@n11.R, B=@n6.R)
    B=TextureSample(UVs=@n12) {samplerSource=SSM_Wrap_WorldGroupSettings, texture=&Scan_M}.R
   )) {constA=0.1, constB=2}
  n14 = VertexColor()
  NamedRerouteDeclaration(None=Multiply(A=Add(A=@n9, B=@n13), B=Multiply(A=@n14, B=@n14))) {name=Opacity, nodeColor=<rgba 0.5703 1 0 1>}
  n15 = VectorParameter() {parameterName=CustomColor, defaultValue=<rgba 0.4021 0.7741 0.8176 4>}
  n16 = ReflectionVectorWS()
  NamedRerouteDeclaration(None=Add(A=Multiply(A=@n15.RGB) {constB=3}, B=Multiply(A=@n16) {constB=0})) {name=Emissive, nodeColor=<rgba 1 0 0.8906 1>}
  NamedRerouteUsage()
  NamedRerouteUsage()
  NamedRerouteUsage()
 ]
}
