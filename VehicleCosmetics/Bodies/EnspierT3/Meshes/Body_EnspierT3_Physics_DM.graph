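# Physics asset for the EnspierT3 vehicle body mesh in VehicleCosmetics, providing its simplified collision shape.
# A single body on the root bone uses one box, offset forward along X and raised above the bone to approximate the car's volume.
# There are no constraints, so the asset defines no jointed or separately simulated parts.
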PhysicsAsset /VehicleCosmetics/Bodies/EnspierT3/Meshes/Body_EnspierT3_Physics_DM
bodies {root [{shapeType=Box, extentX=444, extentY=216, extentZ=116, shapeName=root_box, center=<37 0 90>}]}
constraints []
